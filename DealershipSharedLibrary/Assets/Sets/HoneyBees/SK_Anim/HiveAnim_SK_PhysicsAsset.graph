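# Physics asset for the HiveAnim_SK skeletal mesh in the HoneyBees set, apparently an animated beehive.
# It holds a single capsule body on Bone_003, offset and rotated from the bone origin.
# It defines no constraints, so there is no jointed ragdoll chain, only one simple collision shape for the mesh.
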PhysicsAsset /DealershipSharedLibrary/Assets/Sets/HoneyBees/SK_Anim/HiveAnim_SK_PhysicsAsset
bodies {Bone_003 [{shapeType=Capsule, length=21.648, radius=21.2215, center=<-1.5974 -21.2258 0.2573>, rotation=<rot 0 117.2045 -85.0132>}]}
constraints []
